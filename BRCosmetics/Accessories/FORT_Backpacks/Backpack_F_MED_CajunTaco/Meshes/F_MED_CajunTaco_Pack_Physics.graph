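# Physics asset for the CajunTaco backpack cosmetic, defining capsule bodies and joints for a figure with torso, pelvis, arms and legs.
# The torso rotates freely on the MSTR_body anchor, which has a large offset capsule, and the pelvis hangs off it with narrow limits.
# Small MSTR limb anchors lock angularly to torso or pelvis, each rooting a four-segment chain with limited swing and locked or zero-degree twist.
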PhysicsAsset /BRCosmetics/Accessories/FORT_Backpacks/Backpack_F_MED_CajunTaco/Meshes/F_MED_CajunTaco_Pack_Physics
bodies {
 MSTR_arm_l [{shapeType=Capsule, length=0.505, radius=0.505, shapeName=MSTR_arm_l_capsule, rotation=<rot 0 -90 0>}]
 MSTR_arm_r [{shapeType=Capsule, length=0.505, radius=0.505, shapeName=MSTR_arm_r_capsule, rotation=<rot 0 -90 0>}]
 MSTR_body [{shapeType=Capsule, length=3.8, radius=50.7, shapeName=MSTR_body_capsule, center=<7.57 56.4 -14.66>, rotation=<rot 0 -90 0>}]
 MSTR_leg_l [{shapeType=Capsule, length=0.505, radius=0.505, shapeName=MSTR_leg_l_capsule, rotation=<rot 0 -90 0>}]
 MSTR_leg_r [{shapeType=Capsule, length=0.505, radius=0.505, shapeName=MSTR_leg_r_capsule, rotation=<rot 0 -90 0>}]
 dyn_arm_01_l [{shapeType=Capsule, length=4.7056, radius=1.6926, shapeName=dyn_arm_01_l_capsule, center=<2 0 0>, rotation=<rot 90 0 0>}]
 dyn_arm_01_r [{shapeType=Capsule, length=4.02, radius=1.5, shapeName=dyn_arm_01_r_capsule, center=<2 0 0>, rotation=<rot 90 0 0>}]
 dyn_arm_02_l [{shapeType=Capsule, length=2.7339, radius=1.6033, shapeName=dyn_arm_02_l_capsule, center=<2 0 0>, rotation=<rot 90 0 0>}]
 dyn_arm_02_r [{shapeType=Capsule, length=3.3854, radius=1.8999, shapeName=dyn_arm_02_r_capsule, center=<1.5 0 0>, rotation=<rot 90 0 0>}]
 dyn_arm_03_l [{shapeType=Capsule, length=3.1939, radius=1.5945, shapeName=dyn_arm_03_l_capsule, center=<2 0 0>, rotation=<rot 90 0 0>}]
 dyn_arm_03_r [{shapeType=Capsule, length=2.7897, radius=2.4542, shapeName=dyn_arm_03_r_capsule, center=<1.5 0 0>, rotation=<rot 90 0 0>}]
 dyn_arm_04_l [{shapeType=Capsule, length=1.4683, radius=1.7157, shapeName=dyn_arm_04_l_capsule, center=<2 0 0>, rotation=<rot 90 0 0>}]
 dyn_arm_04_r [{shapeType=Capsule, length=3.5662, radius=2.2894, shapeName=dyn_arm_04_r_capsule, center=<4 0 0>, rotation=<rot 90 0 0>}]
 dyn_body [{shapeType=Capsule, length=8.0368, radius=1, shapeName=dyn_body_capsule, center=<1 0 0>, rotation=<rot 90 0 0>}]
 dyn_leg_01_l [{shapeType=Capsule, length=3.9392, radius=1.5, shapeName=dyn_leg_01_l_capsule, center=<2 0 0>, rotation=<rot 90 0 0>}]
 dyn_leg_01_r [{shapeType=Capsule, length=4.4118, radius=1.5, shapeName=dyn_leg_01_r_capsule, center=<2.5 0 0>, rotation=<rot 90 0 0>}]
 dyn_leg_02_l [{shapeType=Capsule, length=3.7653, radius=1.5, shapeName=dyn_leg_02_l_capsule, center=<2 0 0>, rotation=<rot 90 0 0>}]
 dyn_leg_02_r [{shapeType=Capsule, length=4.525, radius=2.1304, shapeName=dyn_leg_02_r_capsule, center=<3.0355 0 0>, rotation=<rot 90 0 0>}]
 dyn_leg_03_l [{shapeType=Capsule, length=3.342, radius=1.7573, shapeName=dyn_leg_03_l_capsule, center=<2.678 0 0>, rotation=<rot 90 0 0>}]
 dyn_leg_03_r [{shapeType=Capsule, length=2.9925, radius=2.3194, shapeName=dyn_leg_03_r_capsule, center=<2 0 0>, rotation=<rot 90 0 0>}]
 dyn_leg_04_l [{shapeType=Capsule, length=5.4185, radius=1.5, shapeName=dyn_leg_04_l_capsule, center=<2.7745 0 0>, rotation=<rot 90 0 0>}]
 dyn_leg_04_r [{shapeType=Capsule, length=3.8025, radius=2.4566, shapeName=dyn_leg_04_r_capsule, center=<3.1615 0 0>, rotation=<rot 90 0 0>}]
 dyn_pelvis [{
   shapeType=Capsule
   length=7.5505
   radius=2.2572
   shapeName=dyn_pelvis_capsule
   center=<2.8734 -0.2163 -0.0002>
   rotation=<rot -18.1805 -84.4719 -5.4594>
  }]
}
constraints [
 {bone1Name=dyn_arm_01_l, bone2Name=MSTR_arm_l, swing1LimitDegrees=40, swing2LimitDegrees=35, twistLimitDegrees=0, twistMotion=Locked}
 {bone1Name=dyn_arm_02_l, bone2Name=dyn_arm_01_l, swing1LimitDegrees=30, swing2LimitDegrees=30, twistLimitDegrees=0, twistMotion=Locked}
 {bone1Name=dyn_arm_03_l, bone2Name=dyn_arm_02_l, swing1LimitDegrees=30, swing2LimitDegrees=30, twistLimitDegrees=0, twistMotion=Locked}
 {bone1Name=dyn_arm_04_l, bone2Name=dyn_arm_03_l, swing1LimitDegrees=35, swing2LimitDegrees=35, twistLimitDegrees=0, twistMotion=Locked}
 {bone1Name=dyn_arm_01_r, bone2Name=MSTR_arm_r, swing1LimitDegrees=50, swing2LimitDegrees=30, twistLimitDegrees=0}
 {bone1Name=dyn_arm_02_r, bone2Name=dyn_arm_01_r, swing1LimitDegrees=35, swing2LimitDegrees=30, twistLimitDegrees=0}
 {bone1Name=dyn_arm_03_r, bone2Name=dyn_arm_02_r, swing1LimitDegrees=30, swing2LimitDegrees=30, twistLimitDegrees=0}
 {bone1Name=dyn_arm_04_r, bone2Name=dyn_arm_03_r, swing1LimitDegrees=35, swing2LimitDegrees=35, twistLimitDegrees=0}
 {bone1Name=dyn_leg_01_l, bone2Name=MSTR_leg_l, swing1LimitDegrees=22, swing2LimitDegrees=25, twistLimitDegrees=0}
 {bone1Name=dyn_leg_02_l, bone2Name=dyn_leg_01_l, swing1LimitDegrees=20, swing2LimitDegrees=20, twistLimitDegrees=0}
 {bone1Name=dyn_leg_03_l, bone2Name=dyn_leg_02_l, swing1LimitDegrees=20, swing2LimitDegrees=20, twistLimitDegrees=0}
 {bone1Name=dyn_leg_04_l, bone2Name=dyn_leg_03_l, swing1LimitDegrees=20, swing2LimitDegrees=20, twistLimitDegrees=0}
 {bone1Name=dyn_leg_01_r, bone2Name=MSTR_leg_r, swing1LimitDegrees=28, swing2LimitDegrees=25, twistLimitDegrees=0}
 {bone1Name=dyn_leg_02_r, bone2Name=dyn_leg_01_r, swing1LimitDegrees=20, swing2LimitDegrees=20, twistLimitDegrees=0}
 {bone1Name=dyn_leg_03_r, bone2Name=dyn_leg_02_r, swing1LimitDegrees=25, swing2LimitDegrees=25, twistLimitDegrees=0}
 {bone1Name=dyn_leg_04_r, bone2Name=dyn_leg_03_r, swing1LimitDegrees=25, swing2LimitDegrees=25, twistLimitDegrees=0}
 {
  bone1Name=dyn_body
  bone2Name=MSTR_body
  swing1LimitDegrees=45
  swing1Motion=Free
  swing2LimitDegrees=45
  swing2Motion=Free
  twistLimitDegrees=45
  twistMotion=Free
 }
 {bone1Name=dyn_pelvis, bone2Name=dyn_body, swing1LimitDegrees=15, swing2LimitDegrees=4, twistLimitDegrees=0}
 {
  bone1Name=MSTR_arm_r
  bone2Name=dyn_body
  swing1LimitDegrees=45
  swing1Motion=Locked
  swing2LimitDegrees=45
  swing2Motion=Locked
  twistLimitDegrees=45
  twistMotion=Locked
 }
 {
  bone1Name=MSTR_arm_l
  bone2Name=dyn_body
  swing1LimitDegrees=45
  swing1Motion=Locked
  swing2LimitDegrees=45
  swing2Motion=Locked
  twistLimitDegrees=45
  twistMotion=Locked
 }
 {
  bone1Name=MSTR_leg_l
  bone2Name=dyn_pelvis
  swing1LimitDegrees=45
  swing1Motion=Locked
  swing2LimitDegrees=45
  swing2Motion=Locked
  twistLimitDegrees=45
  twistMotion=Locked
 }
 {
  bone1Name=MSTR_leg_r
  bone2Name=dyn_pelvis
  swing1LimitDegrees=45
  swing1Motion=Locked
  swing2LimitDegrees=45
  swing2Motion=Locked
  twistLimitDegrees=45
  twistMotion=Locked
 }
]
