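# AnimBlueprint for the CajunTaco cosmetic backpack that adds procedural secondary motion to the pack.
# An additive offset on the pack bone, mapped each update from spine_01-to-spine_05 bend, feeds Anim Dynamics on dyn_pelvis, a RigidBody node, then an Anim Dynamics chain from dyn_chest to dyn_head.
# A deprecated gravity-override function supplies gravity and RigidBody sim-space settings from per-movement-state presets.
# The front-end flag switches simulation space and an alpha value.
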AnimBlueprint /BRCosmetics/Accessories/FORT_Backpacks/Backpack_F_MED_CajunTaco/Meshes/F_MED_CajunTaco_Pack_AnimBP
target_skeleton=&/F_MED_CajunTaco_Pack_Skeleton
graphs [
 AnimGraph {nodes [
   n1 = "Transform (Modify) Bone - Bone: pack"(
    ComponentPose=Local To Component(LocalPose=Use cached pose 'Input_pose'())
    Translation=Get backpack_translation()
    Rotation=Get backpack_rotate()
   ) {
    rotationMode=BMM_Additive
    rotationSpace=BCS_BoneSpace
    translationMode=BMM_Additive
    translationSpace=BCS_BoneSpace
    alphaBoolBlend {blendOption=Linear}
    boneToModify {boneName=pack}
   }
   n2 = Anim Dynamics - Bone: dyn_pelvis!(SimulationSpace:AnimPhysSimSpaceType Enum, ComponentPose=@n1, Alpha=1, GravityOverride:Vector="0,0,0", ExternalForce="0,0,0") -> (Pose) {
    angularDampingOverride=0.975
    angularSpringConstant=1000
    bAngularSpring=true
    bOverrideAngularDamping=true
    bUseGravityOverride=true
    gravityScale=0.35
    simulationSpace=RootRelative
    alphaBoolBlend {blendOption=Linear}
    boundBone {boneName=dyn_pelvis}
    componentAppliedLinearAccClamp=<3000 3000 3000>
    componentLinearAccScale=<0.3 0.3 0.3>
    componentLinearVelScale=<0.4 0.4 0.4>
    gravityOverride {z=-980}
    physicsBodyDefinitions [{
      boundBone {boneName=dyn_pelvis}
      boxExtents=<15 15 15>
      constraintSetup {angularLimitsMax=<0 8 3>, angularLimitsMin=<0 -8 -20>, angularTarget=<1 0 0>}
      localJointOffset=<3 0 0>
     }]
   }
   n3 = RigidBody!(
    ComponentPose=@n2
    Alpha
    OverrideWorldGravity:Vector="0,0,0"
    SimSpaceSettings:Sim Space Settings Structure=(WorldAlpha=0.35,VelocityScaleZ=0.2,DampingAlpha=1,MaxLinearVelocity=600,MaxAngularVelocity=9,MaxLinearAcceleration=2500,MaxAngularAcceleration=38,ExternalLinearDragV=(X=0.1,Y=0.1,Z=0.1),ExternalLinearVelocity=(X=0,Y=0,Z=0),ExternalAngularVelocity=(X=0,Y=0,Z=0))
   ) -> (Pose) {
    bClampLinearTranslationLimitToRefPose=true
    bOverrideWorldGravity=true
    bUseExternalClothCollision=true
    overridePhysicsAsset=&/F_MED_CajunTaco_Pack_Physics
    simulationSpace=BaseBoneSpace
    alphaBoolBlend {blendOption=Linear}
    baseBoneRef {boneName=MSTR_body}
    componentAppliedLinearAccClamp=<800 800 800>
    componentLinearAccScale=<0.3 0.3 0.2>
    componentLinearVelScale=<0.3 0.3 0.2>
    simSpaceSettings {
     maxAngularAcceleration=38
     maxAngularVelocity=9
     maxLinearAcceleration=2500
     maxLinearVelocity=600
     velocityScaleZ=0.2
     worldAlpha=0.35
     externalLinearDragV=<0.1 0.1 0.1>
    }
   }
   n4 = "Anim Dynamics - Chain: dyn_chest -> dyn_head"!(SimulationSpace:AnimPhysSimSpaceType Enum, ComponentPose=@n3, Alpha=1, GravityOverride:Vector="0,0,0", ExternalForce="0,0,0") -> (Pose) {
    angularDampingOverride=0.999
    angularSpringConstant=2000
    bAngularSpring=true
    bChain=true
    bOverrideAngularDamping=true
    bUseGravityOverride=true
    gravityScale=0.35
    simulationSpace=RootRelative
    alphaBoolBlend {blendOption=Linear}
    boundBone {boneName=dyn_chest}
    chainEnd {boneName=dyn_head}
    componentAppliedLinearAccClamp=<3000 3000 3000>
    componentLinearAccScale=<0.25 0.25 0.25>
    componentLinearVelScale=<0.25 0.25 0.25>
    gravityOverride {z=-980}
    physicsBodyDefinitions [
     {
      boundBone {boneName=dyn_chest}
      boxExtents=<15 25 30>
      constraintSetup {angularLimitsMax=<1 5 4>, angularLimitsMin=<-1 -5 -8>, angularTarget=<1 0 0>}
      localJointOffset=<3 0 0>
     }
     {
      boundBone {boneName=dyn_head}
      boxExtents=<20 25 30>
      constraintSetup {angularLimitsMax=<1 3 4>, angularLimitsMin=<-1 -3 -12>, angularTarget=<1 0 0>}
      localJointOffset=<3 0 0>
     }
    ]
   }
   Output Pose(Result=Component To Local(ComponentPose=@n4)) {layerGroup=None}
   Save cached pose 'Input_pose'(Pose=Input Pose()) {cacheName=Input_pose}
  ]}
 EventGraph {nodes [
   Event BlueprintUpdateAnimation()
   > n1 = Sequence!(execute) -> (then_0, then_1, then_2:Exec)
   > n2 = Gravity Override State Function DEPRECATED(
    CharacterPartAnimInstance=Self-Reference()
    GameTimeInSeconds=GetGameTimeInSeconds()
    GravityOverrideParams=Get Gravity Override Params()
    RBAN Params=Get RBAN Params()
   )
   > Set Sim Space Settings(Sim Space Settings:Sim Space Settings Structure=Break RBANSetting Struct!(RBANSettingStruct:"RBANSetting Struct Structure (by ref)"=@n2.RBAN SettingsOut) -> (
     ComponentLinearAccScale_28_#1:Vector=~
     ComponentLinearVelScale_31_#2:Vector=~
     SimSpaceSettings_34_#3:Sim Space Settings Structure
    ) {is=title K2Node_BreakStruct}.SimSpaceSettings_34_#3) -> (Output_Get:Sim Space Settings Structure)
   > Set GravityOverride(GravityOverride=@n2)
   n3 = GetOwnerSkeletalMeshComponent()
   n4 = BreakRotator(InRot=BreakTransform(InTransform=GetDeltaTransformFromRefPose(self=@n3, BoneName=spine_05, BaseName=spine_01)).Rotation)
   Set backpack_rotate(execute=@n1.then_1, backpack_rotate=MakeRotator(Roll=MapRangeClamped(Value=@n4.Yaw, InRangeA=0, InRangeB=-30, OutRangeB=45)))
   > Set backpack_translation(backpack_translation=MakeVector(Y=MapRangeClamped(Value=@n4.Yaw, InRangeA=0, InRangeB=-30, OutRangeA=-1.2, OutRangeB=-3)))
   > n5 = Branch(Condition=Get bIsInFrontEnd())
   > Set SimlationSpace(SimlationSpace:AnimPhysSimSpaceType Enum=World) -> (Output_Get:AnimPhysSimSpaceType Enum)
   > Set AlphaFE(AlphaFE:double=0.15) -> (Output_Get:double)
   Set SimlationSpace(execute=@n5.else)
   > Set AlphaFE(AlphaFE=1)
   TryGetPawnOwner()
   CustomEvent!() -> (OutputDelegate, then, DeltaTimeX:float=~) {is=title K2Node_CustomEvent}
  ]}
]
variables [
 AlphaFE {type=number}
 backpack_rotate {type=Rotator}
 backpack_translation {type=Vector}
 Gravity Override Params {
  type=GravityOverrideParamsStruct
  default {skydive_FloatingDown {windFrequency=<1.5 5 6>}, skydive_Parachuting {windAmplitude=<100 -250 200>}}
 }
 GravityOverride {type=Vector}
 RBAN Params {
  type=RBANParamsStruct
  default {
   dBNO {simSpaceSettings {
     maxAngularAcceleration=1000
     maxAngularVelocity=3
     maxLinearAcceleration=300
     maxLinearVelocity=700
     velocityScaleZ=0
     externalLinearDragV=<0 0.5 0>
    }}
   dBNO_Carried {simSpaceSettings {
     maxAngularAcceleration=1000
     maxAngularVelocity=3
     maxLinearAcceleration=300
     maxLinearVelocity=700
     velocityScaleZ=0
     externalLinearDragV=<0 0.5 0>
    }}
   emoteOrMelee {
    componentLinearAccScale=<0.3 0.3 0.2>
    componentLinearVelScale=<0.3 0.3 0.2>
    simSpaceSettings {
     maxAngularAcceleration=20
     maxAngularVelocity=5
     maxLinearAcceleration=1800
     maxLinearVelocity=400
     velocityScaleZ=0.2
     worldAlpha=0.75
     externalLinearDragV=<0.1 0.1 0.1>
    }
   }
   front_End {simSpaceSettings {
     dampingAlpha=0
     maxAngularAcceleration=30
     maxAngularVelocity=9
     maxLinearAcceleration=2800
     maxLinearVelocity=750
     velocityScaleZ=0
     worldAlpha=0.35
     externalLinearDragV=<0.2 0.2 0.2>
    }}
   onGround_Crouch_Moving {
    componentLinearAccScale=<0.3 0.3 0.2>
    componentLinearVelScale=<0.3 0.3 0.2>
    simSpaceSettings {
     maxAngularAcceleration=20
     maxAngularVelocity=5
     maxLinearAcceleration=1800
     maxLinearVelocity=400
     velocityScaleZ=0.15
     externalLinearDragV=<0.1 0.1 0.1>
    }
   }
   onGround_Falling {
    componentLinearAccScale=<0.3 0.3 0.2>
    componentLinearVelScale=<0.3 0.3 0.2>
    simSpaceSettings {
     maxAngularAcceleration=20
     maxAngularVelocity=5
     maxLinearAcceleration=1800
     maxLinearVelocity=400
     velocityScaleZ=0.2
     worldAlpha=0.7
     externalLinearDragV=<0.1 0.1 0.1>
    }
   }
   onGround_Moving {
    componentLinearAccScale=<0.3 0.3 0.2>
    componentLinearVelScale=<0.3 0.3 0.2>
    simSpaceSettings {
     maxAngularAcceleration=20
     maxAngularVelocity=5
     maxLinearAcceleration=1800
     maxLinearVelocity=400
     velocityScaleZ=0.17
     worldAlpha=0.75
     externalLinearDragV=<0.1 0.1 0>
    }
   }
   onGround_Standing {
    componentLinearAccScale=<0.3 0.3 0.2>
    componentLinearVelScale=<0.3 0.3 0.2>
    simSpaceSettings {
     maxAngularAcceleration=38
     maxAngularVelocity=9
     maxLinearAcceleration=1800
     maxLinearVelocity=400
     velocityScaleZ=0.17
     worldAlpha=0.75
     externalLinearDragV=<0.1 0.1 0>
    }
   }
   rail_Grinding {simSpaceSettings {
     maxAngularAcceleration=35
     maxAngularVelocity=9
     maxLinearAcceleration=1000
     maxLinearVelocity=200
     velocityScaleZ=0.25
     worldAlpha=0.5
     externalLinearDragV=<0.2 0.2 0.2>
    }}
   riding_Moving {simSpaceSettings {
     maxAngularAcceleration=1000
     maxAngularVelocity=3
     maxLinearAcceleration=300
     maxLinearVelocity=700
     velocityScaleZ=0
     externalLinearDragV=<0 0.5 0>
    }}
   riding_Moving_Targeting {simSpaceSettings {
     maxAngularAcceleration=1000
     maxAngularVelocity=3
     maxLinearAcceleration=300
     maxLinearVelocity=700
     velocityScaleZ=0
     externalLinearDragV=<0 0.5 0>
    }}
   riding_Standing {simSpaceSettings {
     maxAngularAcceleration=1000
     maxAngularVelocity=3
     maxLinearAcceleration=300
     maxLinearVelocity=700
     velocityScaleZ=0
     externalLinearDragV=<0 0.5 0>
    }}
   skydive_DivingDown {
    componentLinearAccScale=<0.3 0.3 0.2>
    componentLinearVelScale=<0.3 0.3 0.2>
    simSpaceSettings {
     maxAngularAcceleration=20
     maxAngularVelocity=5
     maxLinearAcceleration=1800
     maxLinearVelocity=400
     velocityScaleZ=0.15
     worldAlpha=0.5
     externalLinearDragV=<0.1 0.1 0.1>
    }
   }
   skydive_FloatingDown {
    componentLinearAccScale=<0.3 0.3 0.2>
    componentLinearVelScale=<0.3 0.3 0.2>
    simSpaceSettings {
     maxAngularAcceleration=20
     maxAngularVelocity=5
     maxLinearAcceleration=1800
     maxLinearVelocity=400
     velocityScaleZ=0.15
     worldAlpha=0.4
     externalLinearDragV=<0.1 0.1 0.1>
    }
   }
   skydive_GoingUp {
    componentLinearAccScale=<0.3 0.3 0.2>
    componentLinearVelScale=<0.3 0.3 0.2>
    simSpaceSettings {
     maxAngularAcceleration=20
     maxAngularVelocity=5
     maxLinearAcceleration=1800
     maxLinearVelocity=400
     velocityScaleZ=0.2
     externalLinearDragV=<0.1 0.1 0.1>
    }
   }
   skydive_Parachuting {
    componentLinearAccScale=<0.3 0.3 0.2>
    componentLinearVelScale=<0.3 0.3 0.2>
    simSpaceSettings {
     maxAngularAcceleration=20
     maxAngularVelocity=5
     maxLinearAcceleration=1800
     maxLinearVelocity=400
     velocityScaleZ=0.2
     worldAlpha=0.5
     externalLinearDragV=<0.1 0.1 0.1>
    }
   }
   skydive_Surfing {simSpaceSettings {
     maxAngularAcceleration=1000
     maxAngularVelocity=3
     maxLinearAcceleration=300
     maxLinearVelocity=700
     velocityScaleZ=0
     externalLinearDragV=<0 0.5 0>
    }}
   swimming {simSpaceSettings {
     maxAngularAcceleration=1000
     maxAngularVelocity=3
     maxLinearAcceleration=300
     maxLinearVelocity=700
     velocityScaleZ=0
     externalLinearDragV=<0 0.5 0>
    }}
   tactical_Sprint {
    componentLinearAccScale=<0.3 0.3 0.2>
    componentLinearVelScale=<0.3 0.3 0.2>
    simSpaceSettings {
     maxAngularAcceleration=35
     maxAngularVelocity=9
     maxLinearAcceleration=1800
     maxLinearVelocity=400
     velocityScaleZ=0.25
     worldAlpha=0.75
     externalLinearDragV=<0 0.15 0>
    }
   }
   ziplining {
    componentLinearAccScale=<0.2 0.2 0.2>
    componentLinearVelScale=<0.2 0.2 0.2>
    simSpaceSettings {
     maxAngularAcceleration=1000
     maxAngularVelocity=3
     maxLinearAcceleration=300
     maxLinearVelocity=700
     velocityScaleZ=0
     worldAlpha=0.2
     externalLinearDragV=<0 0.5 0>
    }
   }
  }
 }
 Sim Space Settings {
  type=SimSpaceSettings
  default {
   dampingAlpha=1
   maxAngularAcceleration=30
   maxAngularVelocity=3
   maxLinearAcceleration=300
   maxLinearVelocity=700
   velocityScaleZ=0
   worldAlpha=1
   externalLinearDragV=<0 0.5 0>
  }
 }
 SimlationSpace {type=AnimPhysSimSpaceType}
]
